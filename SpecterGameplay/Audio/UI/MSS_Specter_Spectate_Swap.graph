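# One-shot stereo MetaSound source that plays the UI sound for swapping in spectate mode.
# On play, a single stereo Wave Player plays the Delulu_Specter_Spectate_Swap wave straight to the left and right outputs, with no processing or variation.
# The wave player's finish trigger fires the source's OnFinished output, ending the sound.
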MetaSoundSource /SpecterGameplay/Audio/UI/MSS_Specter_Spectate_Swap
inputs [UE.Source.OnPlay {type=Trigger}]
interfaces [UE.OutputFormat.Stereo:1.0, UE.Source.OneShot:1.0, UE.Source:1.1]
outputs [
 "UE.OutputFormat.Stereo.Audio:0" {type=Audio}
 "UE.OutputFormat.Stereo.Audio:1" {type=Audio}
 UE.Source.OneShot.OnFinished {type=Trigger}
]
pages [Default {
  input_defaults {UE.Source.OnPlay=null}
  nodes [WavePlayer_0 = UE.Wave Player:Stereo(Play=^UE.Source.OnPlay) {also {defaults {Wave Asset=&/Delulu_Specter_Spectate_Swap}, version [1,1]}}]
  links [
   {dst="@UE.Source.OneShot.OnFinished", src=WavePlayer_0.On Finished}
   {dst="@UE.OutputFormat.Stereo.Audio:0", src=WavePlayer_0.Out Left}
   {dst="@UE.OutputFormat.Stereo.Audio:1", src=WavePlayer_0.Out Right}
  ]
 }]
properties {assetImportData=&self:AssetImportData, numChannels=2, outputFormat=Stereo, sampleRate=48000}
